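# Animation Blueprint for the female ChromeSiren character body, copying the attached parent mesh's pose and passing it through a ChromeSiren Control Rig.
# Update events turn crouch, targeting, sliding, parachute, emote and front-end states into rig inputs, smoothing crouched strafing with FInterpTo.
# A cosmetic variant lookup on the owning pawn sets the helmet flag.
# Weapon-change events classify the core animation into rifle, pistol or melee, feeding rifle and pistol weights to the rig.
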AnimBlueprint /BRCosmetics/Characters/Player/Female/Medium/Bodies/F_MED_ChromeSiren/Meshes/F_MED_ChromeSiren_AnimBP
target_skeleton=&/F_MED_ChromeSiren_Skeleton
graphs [
 AnimGraph {nodes [
   n1 = Control Rig!(
    Source=Copy Pose From Mesh() {bUseAttachedParent=true}
    Alpha
    Crouch walk:double=Get CrouchWalk() -> (CrouchWalk:double)
    Targeting:double=Get Targeting() -> (Targeting:double)
    Crouch Alpha:double=Get IsCrouching() -> (IsCrouching:double)
    Only crouch:bool=Get OnlyCrouch() -> (OnlyCrouch:bool)
    parachute open:double=Get Parachute Open() -> (Parachute Open:double)
    emote:double=Get Emote() -> (Emote:double)
    Not emote:double=Get Not emote() -> (Not emote:double)
    Pistol:double=Get pistol() -> (pistol:double)
    Riffle:double=Get Rifle() -> (Rifle:double)
    isCrouch Float:double=Get isCrouch float() -> (isCrouch float:double)
    Sky Diving:bool=Get bIsSkydiveFloating()
    HelmetOn:bool=Get Helmet() -> (Helmet:bool)
    FrontEnd:bool=Get FrontEnd Idle() -> (FrontEnd Idle:bool)
   ) -> (Pose) {controlRigAssetReference {blueprintRigClass=&/F_MED_ChromeSiren_CtrlRig.F_MED_ChromeSiren_CtrlRig_C}}
   Output Pose(Result=@n1) {layerGroup=None}
  ]}
 EventGraph {nodes [
   n1 = Event BlueprintUpdateAnimation()
   > Set Delta Time X(Delta Time X=@n1.DeltaTimeX)
   > n2 = Sequence!(execute) -> (then_0, then_1, then_2:Exec, then_3:Exec, then_4:Exec, then_5:Exec, then_6:Exec)
   n3 = Get bIsCrouching()
   n4 = Get bIsMoving2D()
   > Set CrouchWalk(CrouchWalk:double="To Float (Boolean)"(InBool=AND Boolean(A=@n3, B=@n4))) -> (Output_Get:double)
   Set Targeting(execute=@n2.then_1, Targeting:double="To Float (Boolean)"(InBool=Get bIsTargeting())) -> (Output_Get:double)
   n5 = OR Boolean(A=Get bIsSliding(), B=Get bIsBaseballSliding())
   Set OnlyCrouch(execute=@n2.then_2, OnlyCrouch:bool=AND Boolean(A=Get bIsCrouching(), B=NOT Boolean(A=@n5))) -> (Output_Get:bool)
   > Set Parachute Open(Parachute Open:double="To Float (Boolean)"(InBool=Get bIsParachuteOpened())) -> (Output_Get:double)
   > Set Emote(Emote:double="To Float (Boolean)"(InBool=Get bIsPlayingEmote())) -> (Output_Get:double)
   n6 = Get bIsPlayingEmote()
   > Set Not emote(Not emote:double="To Float (Boolean)"(InBool=NOT Boolean(A=@n6))) -> (Output_Get:double)
   n7 = Cast To FortPlayerPawn(execute=@n2.then_4, Object=GetOwningActor())
   > n8 = GetVariantChannelActiveIndexForItemActor!(
    execute
    Item:Athena Cosmetic Item Definition Object Reference=/BRCosmetics/Athena/Items/Cosmetics/Characters/Character_ChromeSiren
    Actor:Actor Object Reference=@n7
    Channel:Gameplay Tag Structure=(TagName="Cosmetics.Variant.Channel.Parts")
    VariantList:Array of Gameplay Tag Structures=Make Array("[0]"=(TagName="Cosmetics.Variant.Property.Stage2"))
    bErrorIfChannelMissing:bool=false
    bErrorIfTagNotInArray:bool=false
   ) -> (then, Index:int=0, bIndexIsValid:bool, bFoundChannel:bool)
   > n9 = Branch(Condition=@n8.bIndexIsValid)
   > Set Helmet(Helmet:bool=true) -> (Output_Get:bool)
   Set Helmet(execute=@n9.else, Helmet=false)
   n10 = Get bIsPlayingEmote()
   Set FrontEnd Idle(execute=@n2.then_6, FrontEnd Idle:bool=AND Boolean(A=Get bIsInFrontEnd(), B=NOT Boolean(A=@n10))) -> (Output_Get:bool)
   n11 = PlayerMovement(execute=@n2.then_3, AnimInstance=Self-Reference(), XDotProductCoefficient=100, YDotProductCoefficient=100, ZDotProductCoefficient=100)
   n12 = BreakVector(InVec=MakeVector(X=@n11, Y=@n11.YMovement))
   > Set moveLeft bool(moveLeft bool:bool="float < float"(A=@n12, B=-50)) -> (Output_Get:bool)
   > Set moveRight bool(moveRight bool:bool="float > float"(A=@n12, B=50)) -> (Output_Get:bool)
   n13 = AND Boolean!(
    A=Get bIsCrouching()
    B=Get bIsMoving2D()
    C:bool=OR Boolean(A=Get moveLeft bool() -> (moveLeft bool:bool), B=Get moveRight bool() -> (moveRight bool:bool))
   ) -> (ReturnValue)
   > Set isCrouch float(isCrouch float:double=FInterpTo(Current=Get isCrouch float(), Target="To Float (Boolean)"(InBool=@n13), DeltaTime=Get Delta Time X(), InterpSpeed=10)) -> (Output_Get:double)
   Event OnBeginCrouching()
   > Set IsCrouching(IsCrouching:double=1) -> (Output_Get:double)
   Event OnEndCrouching()
   > Set IsCrouching(IsCrouching=0)
   Event OnWeaponChanged()
   > n14 = Cast To FortPlayerAnimInstance(Object=GetAnimInstance(self=GetOwnerSkeletalMeshComponent()))
   > Set pistol(pistol:double=0) -> (Output_Get:double)
   > Set Melee(Melee:double=0) -> (Output_Get:double)
   > Set Rifle(Rifle:double=0) -> (Output_Get:double)
   > n15 = Switch on EFortWeaponCoreAnimation!(execute, Selection:EFortWeaponCoreAnimation Enum=Get WeaponCoreAnim(self=@n14)) -> (
    Melee:Exec
    Pistol:Exec
    Shotgun:Exec
    PaperBlueprint:Exec
    Rifle:Exec
    MeleeOneHand:Exec
    MachinePistol:Exec
    RocketLauncher:Exec
    GrenadeLauncher:Exec
    GoingCommando:Exec
    AssaultRifle:Exec
    TacticalShotgun:Exec
    SniperRifle:Exec
    TrapPlacement:Exec
    ShoulderLauncher:Exec
    AbilityDecoTool:Exec
    Crossbow:Exec
    C4:Exec
    RemoteControl:Exec
    DualWield:Exec
    AR_BullPup:Exec
    AR_ForwardGrip:Exec
    MedPackPaddles:Exec
    SMG_P90:Exec
    AR_DrumGun:Exec
    Consumable_Small:Exec
    Consumable_Large:Exec
    Balloon:Exec
    MountedTurret:Exec
    CreativeTool:Exec
    ExplosiveBow:Exec
    AshtonIndigo:Exec
    AshtonChicago:Exec
    MeleeDualWield:Exec
    Unarmed:Exec
   )
   Set Rifle(
    execute=[
     @n15.AR_BullPup
     @n15.AR_DrumGun
     @n15.AR_ForwardGrip
     @n15.AssaultRifle
     @n15.Crossbow
     @n15.GrenadeLauncher
     @n15.Rifle
     @n15.RocketLauncher
     @n15.SMG_P90
     @n15.Shotgun
     @n15.SniperRifle
     @n15.TacticalShotgun
    ]
    Rifle=1
   )
   > Set Only_Rifle(Only_Rifle:bool=1) -> (Output_Get:bool)
   > Set Only_Pistol(Only_Pistol:bool=false) -> (Output_Get:bool)
   Set pistol(execute=[@n15.MachinePistol, @n15.Pistol], pistol=1)
   > Set Only_Rifle(Only_Rifle=false)
   > Set Only_Pistol(Only_Pistol=true)
   Set Melee(execute=@n15.Melee, Melee=1)
   > Set Only_Rifle(Only_Rifle=false)
   > Set Only_Pistol(Only_Pistol=false)
   Reroute Node(InputPin:Wildcard) -> (OutputPin:Wildcard)
  ]}
]
variables [
 Celebration {type=boolean}
 CrouchWalk {type=number}
 customEmotePlatypus {type=boolean}
 Delta Time X {type=number}
 EasternBlocEmote {type=boolean}
 Emote {type=number}
 FrontEnd Idle {type=boolean}
 Helmet {type=boolean}
 isCrouch float {type=number}
 IsCrouching {type=number}
 Melee {type=number}
 moveLeft bool {type=boolean}
 moveRight bool {type=boolean}
 Not emote {type=number}
 Only_Pistol {type=boolean}
 Only_Rifle {type=boolean}
 OnlyCrouch {type=boolean}
 Parachute Open {type=number}
 pistol {type=number}
 Rifle {type=number}
 Targeting {type=number}
]
