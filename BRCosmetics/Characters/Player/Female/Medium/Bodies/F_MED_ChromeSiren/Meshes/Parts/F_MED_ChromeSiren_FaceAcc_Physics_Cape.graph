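# Physics asset for the ChromeSiren medium female face accessory, named Cape but built around a ponytail bone chain and upper-body capsules.
# The ponytail base hangs off the head on a free joint, the first link is locked to it, and the later links have swing limits with very tight twist.
# Capsules on spine, neck, head, clavicles and upper arms form the body collision shapes, joined by free constraints mostly centred on spine_04.
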PhysicsAsset /BRCosmetics/Characters/Player/Female/Medium/Bodies/F_MED_ChromeSiren/Meshes/Parts/F_MED_ChromeSiren_FaceAcc_Physics_Cape
bodies {
 base_ponytail [{shapeType=Capsule, length=2, radius=4, shapeName=base_ponytail_capsule, rotation=<rot 90 0 0>}]
 clavicle_l [{shapeType=Capsule, length=12.6283, radius=7.5, shapeName=clavicle_l_capsule, center=<5 0 -1.6229>, rotation=<rot 0 -90 90>}]
 clavicle_r [{shapeType=Capsule, length=12.2586, radius=7.5, shapeName=clavicle_r_capsule, center=<-5 0 1.6229>, rotation=<rot 0 -90 -90>}]
 dyn_ponytail_01 [{shapeType=Capsule, length=5, radius=5, shapeName=dyn_ponytail_01_capsule, center=<4 0 0>, rotation=<rot 0 90 90>}]
 dyn_ponytail_02 [{shapeType=Capsule, length=5, radius=5, shapeName=dyn_ponytail_02_capsule, center=<5 0 1.5>, rotation=<rot 0 90 90>}]
 dyn_ponytail_03 [{shapeType=Capsule, length=5, radius=4, shapeName=dyn_ponytail_03_capsule, center=<5 0 1>, rotation=<rot 0 90 90>}]
 dyn_ponytail_04 [
  {shapeType=Capsule, length=5, radius=4, shapeName=dyn_ponytail_04_capsule, center=<5 0 0.5>, rotation=<rot 0 90 90>}
  {
   shapeType=Capsule
   length=5.9891
   radius=4.0109
   shapeName=dyn_ponytail_04_capsule
   center=<6.0131 -3.6378 -1.4542>
   rotation=<rot 84.2839 0 0>
  }
 ]
 dyn_ponytail_05 [{shapeType=Capsule, length=5, radius=4, shapeName=dyn_ponytail_05_capsule, center=<4 0 1>, rotation=<rot 0 90 90>}]
 dyn_ponytail_06 [{shapeType=Capsule, length=2.6873, radius=6.3127, shapeName=dyn_ponytail_06_capsule, center=<4 -2.2 1.5>, rotation=<rot 0 90 90>}]
 head [{shapeType=Capsule, length=4.9949, radius=7, shapeName=head_capsule, center=<3.8411 -1.1539 0>, rotation=<rot 90 0 0>}]
 neck_01 [{shapeType=Capsule, length=3.3674, radius=7, shapeName=neck_01_capsule, center=<4 0 0>, rotation=<rot 0 -110 -90>}]
 spine_02 [{shapeType=Capsule, length=15, radius=15, shapeName=spine_02_capsule, center=<0 -5 0>, rotation=<rot 0 -90 0>}]
 spine_03 [{shapeType=Capsule, length=15, radius=15, shapeName=spine_03_capsule, center=<0 -5 0>, rotation=<rot 0 -90 0>}]
 spine_04 [{shapeType=Capsule, length=15, radius=15, shapeName=spine_04_capsule, center=<0 -5 0>, rotation=<rot 0 -90 0>}]
 spine_05 [
  {shapeType=Capsule, length=14.244, radius=15.0041, shapeName=spine_05_capsule, center=<2.1031 -2.2304 0>, rotation=<rot 0 -90 0>}
  {shapeType=Capsule, length=19.0126, radius=6.0155, shapeName=spine_05_capsule, center=<9.4351 -1.8079 0>}
 ]
 upperarm_l [{shapeType=Capsule, length=18.9692, radius=6.3198, shapeName=upperarm_l_capsule, center=<10 0 0>, rotation=<rot 0 -90 90>}]
 upperarm_r [{shapeType=Capsule, length=18.7079, radius=6.5811, shapeName=upperarm_r_capsule, center=<-10 -1.0716 0>, rotation=<rot 0 -90 -90>}]
}
constraints [
 {
  bone1Name=base_ponytail
  bone2Name=head
  swing1LimitDegrees=45
  swing1Motion=Free
  swing2LimitDegrees=45
  swing2Motion=Free
  twistLimitDegrees=45
  twistMotion=Free
 }
 {
  bone1Name=dyn_ponytail_01
  bone2Name=base_ponytail
  swing1LimitDegrees=40
  swing1Motion=Locked
  swing2LimitDegrees=40
  swing2Motion=Locked
  twistLimitDegrees=10
  twistMotion=Locked
 }
 {bone1Name=dyn_ponytail_02, bone2Name=dyn_ponytail_01, swing1LimitDegrees=10, swing2LimitDegrees=10, twistLimitDegrees=2}
 {bone1Name=dyn_ponytail_03, bone2Name=dyn_ponytail_02, swing1LimitDegrees=25, swing2LimitDegrees=35, twistLimitDegrees=2}
 {bone1Name=dyn_ponytail_04, bone2Name=dyn_ponytail_03, swing1LimitDegrees=30, swing2LimitDegrees=40, twistLimitDegrees=2}
 {bone1Name=dyn_ponytail_05, bone2Name=dyn_ponytail_04, swing1LimitDegrees=20, swing2LimitDegrees=30, twistLimitDegrees=2}
 {bone1Name=dyn_ponytail_06, bone2Name=dyn_ponytail_05, swing1LimitDegrees=20, swing2LimitDegrees=20, twistLimitDegrees=2}
 {
  bone1Name=spine_03
  bone2Name=spine_02
  swing1LimitDegrees=45
  swing1Motion=Free
  swing2LimitDegrees=45
  swing2Motion=Free
  twistLimitDegrees=45
  twistMotion=Free
 }
 {
  bone1Name=spine_04
  bone2Name=spine_03
  swing1LimitDegrees=45
  swing1Motion=Free
  swing2LimitDegrees=45
  swing2Motion=Free
  twistLimitDegrees=45
  twistMotion=Free
 }
 {
  bone1Name=spine_05
  bone2Name=spine_04
  swing1LimitDegrees=45
  swing1Motion=Free
  swing2LimitDegrees=45
  swing2Motion=Free
  twistLimitDegrees=45
  twistMotion=Free
 }
 {
  bone1Name=neck_01
  bone2Name=spine_04
  swing1LimitDegrees=45
  swing1Motion=Free
  swing2LimitDegrees=45
  swing2Motion=Free
  twistLimitDegrees=45
  twistMotion=Free
 }
 {
  bone1Name=head
  bone2Name=spine_04
  swing1LimitDegrees=45
  swing1Motion=Free
  swing2LimitDegrees=45
  swing2Motion=Free
  twistLimitDegrees=45
  twistMotion=Free
 }
 {
  bone1Name=clavicle_l
  bone2Name=spine_04
  swing1LimitDegrees=45
  swing1Motion=Free
  swing2LimitDegrees=45
  swing2Motion=Free
  twistLimitDegrees=45
  twistMotion=Free
 }
 {
  bone1Name=upperarm_l
  bone2Name=spine_04
  swing1LimitDegrees=45
  swing1Motion=Free
  swing2LimitDegrees=45
  swing2Motion=Free
  twistLimitDegrees=45
  twistMotion=Free
 }
 {
  bone1Name=clavicle_r
  bone2Name=spine_04
  swing1LimitDegrees=45
  swing1Motion=Free
  swing2LimitDegrees=45
  swing2Motion=Free
  twistLimitDegrees=45
  twistMotion=Free
 }
 {
  bone1Name=upperarm_r
  bone2Name=spine_04
  swing1LimitDegrees=45
  swing1Motion=Free
  swing2LimitDegrees=45
  swing2Motion=Free
  twistLimitDegrees=45
  twistMotion=Free
 }
]
