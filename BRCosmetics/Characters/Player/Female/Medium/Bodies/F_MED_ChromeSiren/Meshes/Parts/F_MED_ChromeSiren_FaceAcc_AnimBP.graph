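# Post-process style hair AnimBlueprint for the F_MED_ChromeSiren FaceAcc skeleton, adding secondary motion to ponytail, bangs and strands.
# Wearing a backpack pins the ponytail with weighted bone constraints; otherwise RigidBody simulates it, with a separate physics asset and pose offsets for cape backblings.
# Chained spring-based AnimDynamics nodes then move bangs and strands.
# A near-duplicate branch retunes angular inheritance and adds a front-end ponytail offset.
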AnimBlueprint /BRCosmetics/Characters/Player/Female/Medium/Bodies/F_MED_ChromeSiren/Meshes/Parts/F_MED_ChromeSiren_FaceAcc_AnimBP
target_skeleton=&/F_MED_ChromeSiren_FaceAcc_Skeleton
graphs [
 AnimGraph {nodes [
   Anim Dynamics - Bone: base_hair_r_01!(ComponentPose, Alpha=1, GravityOverride:Vector="0,0,-980", ExternalForce="0,0,0") -> (Pose) {
    angularDampingOverride=1
    angularSpringConstant=12000
    bAngularSpring=true
    bOverrideAngularDamping=true
    bUseGravityOverride=true
    bUseSphericalLimits=true
    gravityScale=0.35
    simulationSpace=RootRelative
    boundBone {boneName=base_hair_r_01}
    componentAppliedLinearAccClamp=<5000 5000 5000>
    componentLinearAccScale=<1 1 1>
    componentLinearVelScale=<1 1 1>
    gravityOverride {z=-980}
    physicsBodyDefinitions [{
      boundBone {boneName=base_hair_r_01}
      boxExtents=<20 30 30>
      constraintSetup {angularLimitsMax=<0 20 5>, angularLimitsMin=<0 0 -5>, angularTarget=<1 0 0>}
      localJointOffset=<4 0 0>
     }]
    simSpaceSettings {simSpaceAngularAlpha=0.02}
    sphericalLimits [{limitRadius=7, drivingBone {boneName=head}, sphereLocalOffset=<2.0023 5 -1.2487>}]
   }
   Anim Dynamics - Bone: base_hair_r_02!(ComponentPose, Alpha=1, GravityOverride:Vector="0,0,-980", ExternalForce="0,0,0") -> (Pose) {
    angularDampingOverride=1
    angularSpringConstant=12000
    bAngularSpring=true
    bOverrideAngularDamping=true
    bUseGravityOverride=true
    bUseSphericalLimits=true
    gravityScale=0.35
    simulationSpace=RootRelative
    boundBone {boneName=base_hair_r_02}
    componentAppliedLinearAccClamp=<5000 5000 5000>
    componentLinearAccScale=<1 1 1>
    componentLinearVelScale=<1 1 1>
    gravityOverride {z=-980}
    physicsBodyDefinitions [{
      boundBone {boneName=base_hair_r_02}
      boxExtents=<20 30 30>
      constraintSetup {angularLimitsMax=<0 20 5>, angularLimitsMin=<0 0 -5>, angularTarget=<1 0 0>}
      localJointOffset=<4 0 0>
     }]
    simSpaceSettings {simSpaceAngularAlpha=0.02}
    sphericalLimits [{limitRadius=7, drivingBone {boneName=head}, sphereLocalOffset=<-3.2283 3.7702 0>}]
   }
   n1 = RigidBody!(ComponentPose=Local To Component(LocalPose=Use cached pose 'pre sim_1'()), Alpha, OverrideWorldGravity:Vector="0,0,0", ExternalForce) -> (Pose) {
    bOverrideWorldGravity=true
    overridePhysicsAsset=&/F_MED_ChromeSiren_FaceAcc_Physics
    simulationSpace=BaseBoneSpace
    baseBoneRef {boneName=spine_05}
    componentAppliedLinearAccClamp=<1000 1000 1000>
    simSpaceSettings {
     maxAngularAcceleration=50
     maxAngularVelocity=10
     maxLinearAcceleration=1800
     maxLinearVelocity=200
     worldAlpha=1
     externalLinearDragV=<2 2 1>
    }
   }
   n2 = RigidBody(ComponentPose=Local To Component(LocalPose=Use cached pose 'pre sim_1'()), OverrideWorldGravity="0,0,0") {
    bOverrideWorldGravity=true
    overridePhysicsAsset=&/F_MED_ChromeSiren_FaceAcc_Physics_Cape
    simulationSpace=BaseBoneSpace
    baseBoneRef {boneName=spine_05}
    componentAppliedLinearAccClamp=<1000 1000 1000>
    simSpaceSettings {
     maxAngularAcceleration=50
     maxAngularVelocity=10
     maxLinearAcceleration=1800
     maxLinearVelocity=200
     worldAlpha=1
     externalLinearDragV=<2 2 1>
    }
   }
   n3 = Get Is Backbling A Cape() -> (Is Backbling A Cape:bool)
   n4 = "Transform (Modify) Bone - Bone: dyn_ponytail_02"!(ComponentPose=@n2, bAlphaBoolEnabled:bool=@n3, Translation="0,3,0", Rotation="8,0,0", Scale="1,1,1") -> (Pose) {
    alphaInputType=Bool
    rotationMode=BMM_Additive
    rotationSpace=BCS_BoneSpace
    translationMode=BMM_Additive
    translationSpace=BCS_BoneSpace
    boneToModify {boneName=dyn_ponytail_02}
    rotation {pitch=8}
    translation {y=3}
   }
   n5 = "Transform (Modify) Bone - Bone: dyn_ponytail_03"!(ComponentPose=@n4, bAlphaBoolEnabled:bool=@n3, Translation="0,0,0", Rotation="8,0,0", Scale="1,1,1") -> (Pose) {
    alphaInputType=Bool
    rotationMode=BMM_Additive
    rotationSpace=BCS_BoneSpace
    translationMode=BMM_Additive
    translationSpace=BCS_BoneSpace
    boneToModify {boneName=dyn_ponytail_03}
    rotation {pitch=8}
   }
   n6 = "Transform (Modify) Bone - Bone: dyn_ponytail_04"!(ComponentPose=@n5, bAlphaBoolEnabled:bool=@n3, Translation="0,0,0", Rotation="0,0,0", Scale="1,1,1") -> (Pose) {
    alphaInputType=Bool
    rotationMode=BMM_Additive
    rotationSpace=BCS_BoneSpace
    translationMode=BMM_Additive
    translationSpace=BCS_BoneSpace
    boneToModify {boneName=dyn_ponytail_04}
   }
   n7 = "Transform (Modify) Bone - Bone: dyn_ponytail_05"!(ComponentPose=@n6, bAlphaBoolEnabled:bool=@n3, Translation="0,0,0", Rotation="0,0,0", Scale="1,1,1") -> (Pose) {
    alphaInputType=Bool
    rotationMode=BMM_Additive
    rotationSpace=BCS_BoneSpace
    translationMode=BMM_Additive
    translationSpace=BCS_BoneSpace
    boneToModify {boneName=dyn_ponytail_05}
   }
   n8 = Blend Poses by bool(BlendPose_0=Component To Local(ComponentPose=@n7), BlendPose_1=Component To Local(ComponentPose=@n1), bActiveValue=Get Is Backbling A Cape()) {blendType=Linear}
   n9 = "Constraint - dyn_ponytail_01 (P)"!(ComponentPose=Local To Component(LocalPose=Use cached pose 'pre sim_1'()), Alpha=1, ConstraintWeights_0=1, ConstraintWeights_1:float=0) -> (Pose) {boneToModify {boneName=dyn_ponytail_01}, constraintSetup [{targetBone {boneName=head}}], constraintWeights [1,0]}
   n10 = "Constraint - dyn_ponytail_02 (P)"!(ComponentPose=@n9, Alpha=1, ConstraintWeights_0=1, ConstraintWeights_1:float=0) -> (Pose) {boneToModify {boneName=dyn_ponytail_02}, constraintSetup [{targetBone {boneName=head}}], constraintWeights [1,0]}
   n11 = "Constraint - dyn_ponytail_03 (P)"!(ComponentPose=@n10, Alpha=1, ConstraintWeights_0=1, ConstraintWeights_1:float=0) -> (Pose) {boneToModify {boneName=dyn_ponytail_03}, constraintSetup [{targetBone {boneName=head}}], constraintWeights [1,0]}
   n12 = "Constraint - dyn_ponytail_04 (PPPP)"!(
    ComponentPose=@n11
    Alpha=1
    ConstraintWeights_0=0.5232
    ConstraintWeights_1:float=0.2939
    ConstraintWeights_2:float=0.1142
    ConstraintWeights_3:float=0.0688
   ) -> (Pose) {
    boneToModify {boneName=dyn_ponytail_04}
    constraintSetup [{targetBone {boneName=neck_02}}, {targetBone {boneName=head}}, {targetBone {boneName=spine_05}}, {targetBone {boneName=neck_01}}]
    constraintWeights [1,1,1,1]
   }
   n13 = "Constraint - dyn_ponytail_05 (PPPPP)"!(
    ComponentPose=@n12
    Alpha=1
    ConstraintWeights_0=0.4764
    ConstraintWeights_1:float=0.426
    ConstraintWeights_2:float=0.0689
    ConstraintWeights_3:float=0.0186
    ConstraintWeights_4:float=0.0101
   ) -> (Pose) {
    boneToModify {boneName=dyn_ponytail_05}
    constraintSetup [
     {targetBone {boneName=spine_05}}
     {targetBone {boneName=neck_01}}
     {targetBone {boneName=head}}
     {targetBone {boneName=spine_03}}
     {targetBone {boneName=neck_02}}
    ]
    constraintWeights [1,1,1,1,1]
   }
   n14 = "Constraint - dyn_ponytail_06 (PP)"!(ComponentPose=@n13, Alpha=1, ConstraintWeights_0=0.9931, ConstraintWeights_1:float=0.0069) -> (Pose) {
    boneToModify {boneName=dyn_ponytail_06}
    constraintSetup [{targetBone {boneName=spine_05}}, {targetBone {boneName=spine_04}}]
    constraintWeights [1,1]
   }
   n15 = "Transform (Modify) Bone - Bone: dyn_ponytail_04"!(ComponentPose=@n14, Alpha:float=1, Translation="0,0,-1", Rotation="0,0,0", Scale="1,1,1") -> (Pose) {translationMode=BMM_Additive, translationSpace=BCS_BoneSpace, boneToModify {boneName=dyn_ponytail_04}, translation {z=-1}}
   n16 = "Transform (Modify) Bone - Bone: dyn_ponytail_05"!(ComponentPose=@n15, Alpha:float=1, Translation="0,0,-2", Rotation="0,0,0", Scale="1,1,1") -> (Pose) {translationMode=BMM_Additive, translationSpace=BCS_BoneSpace, boneToModify {boneName=dyn_ponytail_05}, translation {z=-2}}
   n17 = "Transform (Modify) Bone - Bone: dyn_ponytail_06"(ComponentPose=@n16, Alpha=1, Translation="0,0,-2", Rotation="0,0,0", Scale="1,1,1") {translationMode=BMM_Additive, translationSpace=BCS_BoneSpace, boneToModify {boneName=dyn_ponytail_06}, translation {z=-2}}
   n18 = Component To Local(ComponentPose=@n17)
   n19 = Get is Wear a backpack() -> (is Wear a backpack:bool)
   n20 = Anim Dynamics - Bone: dyn_bangs_l(
    ComponentPose=Local To Component(LocalPose=Blend Poses by bool(BlendPose_0=@n18, BlendPose_1=@n8, bActiveValue=@n19) {blendType=Linear})
    Alpha=1
    ExternalForce="0,0,0"
   ) {
    angularDampingOverride=1
    angularSpringConstant=12000
    bAngularSpring=true
    bOverrideAngularDamping=true
    bUseGravityOverride=true
    gravityScale=0.35
    simulationSpace=RootRelative
    boundBone {boneName=dyn_bangs_l}
    componentLinearAccScale=<0.5 0.5 0.5>
    componentLinearVelScale=<0.5 0.5 0.5>
    gravityOverride {z=-980}
    physicsBodyDefinitions [{
      boundBone {boneName=dyn_bangs_l}
      boxExtents=<20 30 30>
      constraintSetup {angularLimitsMax=<0 8 2>, angularLimitsMin=<0 -8 -2>, angularTarget=<1 0 0>}
      localJointOffset=<1 0 0>
     }]
    simSpaceSettings {simSpaceAngularAlpha=0.5}
   }
   n21 = Anim Dynamics - Bone: dyn_bangs_r(ComponentPose=@n20, Alpha=1, ExternalForce="0,0,0") {
    angularDampingOverride=1
    angularSpringConstant=12000
    bAngularSpring=true
    bOverrideAngularDamping=true
    bUseGravityOverride=true
    gravityScale=0.35
    simulationSpace=RootRelative
    boundBone {boneName=dyn_bangs_r}
    componentLinearAccScale=<0.5 0.5 0.5>
    componentLinearVelScale=<0.5 0.5 0.5>
    gravityOverride {z=-980}
    physicsBodyDefinitions [{
      boundBone {boneName=dyn_bangs_r}
      boxExtents=<20 30 30>
      constraintSetup {angularLimitsMax=<0 8 2>, angularLimitsMin=<0 -8 -2>, angularTarget=<1 0 0>}
      localJointOffset=<1 0 0>
     }]
    simSpaceSettings {simSpaceAngularAlpha=0.5}
   }
   n22 = Anim Dynamics - Bone: dyn_hair_l_01!(ComponentPose=@n21, Alpha=1, GravityOverride:Vector="0,0,-980", ExternalForce="0,0,0") -> (Pose) {
    angularDampingOverride=1
    angularSpringConstant=12000
    bAngularSpring=true
    bOverrideAngularDamping=true
    bUseGravityOverride=true
    gravityScale=0.35
    simulationSpace=RootRelative
    boundBone {boneName=dyn_hair_l_01}
    componentAppliedLinearAccClamp=<5000 5000 5000>
    componentLinearAccScale=<1 1 1>
    componentLinearVelScale=<1 1 1>
    gravityOverride {z=-980}
    physicsBodyDefinitions [{
      boundBone {boneName=dyn_hair_l_01}
      boxExtents=<20 30 30>
      constraintSetup {angularLimitsMax=<0 0 5>, angularLimitsMin=<0 -15 -5>, angularTarget=<1 0 0>}
      localJointOffset=<4 0 0>
     }]
    simSpaceSettings {simSpaceAngularAlpha=0.02}
   }
   n23 = Anim Dynamics - Bone: dyn_hair_l_02!(ComponentPose=@n22, Alpha=1, GravityOverride:Vector="0,0,-980", ExternalForce="0,0,0") -> (Pose) {
    angularDampingOverride=1
    angularSpringConstant=12000
    bAngularSpring=true
    bOverrideAngularDamping=true
    bUseGravityOverride=true
    gravityScale=0.35
    simulationSpace=RootRelative
    boundBone {boneName=dyn_hair_l_02}
    componentAppliedLinearAccClamp=<5000 5000 5000>
    componentLinearAccScale=<1 1 1>
    componentLinearVelScale=<1 1 1>
    gravityOverride {z=-980}
    physicsBodyDefinitions [{
      boundBone {boneName=dyn_hair_l_02}
      boxExtents=<20 30 30>
      constraintSetup {angularLimitsMax=<0 0 5>, angularLimitsMin=<0 -15 -5>, angularTarget=<1 0 0>}
      localJointOffset=<4 0 0>
     }]
    simSpaceSettings {simSpaceAngularAlpha=0.02}
   }
   n24 = Anim Dynamics - Bone: dyn_dyn_hair_f_m_01!(ComponentPose=@n23, Alpha=1, GravityOverride:Vector="0,0,-980", ExternalForce="0,0,0") -> (Pose) {
    angularDampingOverride=1
    angularSpringConstant=12000
    bAngularSpring=true
    bOverrideAngularDamping=true
    bUseGravityOverride=true
    gravityScale=0.35
    simulationSpace=RootRelative
    boundBone {boneName=dyn_dyn_hair_f_m_01}
    componentAppliedLinearAccClamp=<5000 5000 5000>
    componentLinearAccScale=<1 1 1>
    componentLinearVelScale=<1 1 1>
    gravityOverride {z=-980}
    physicsBodyDefinitions [{
      boundBone {boneName=dyn_dyn_hair_f_m_01}
      boxExtents=<20 30 30>
      constraintSetup {angularLimitsMax=<0 0 10>, angularLimitsMin=<0 -20 -10>, angularTarget=<1 0 0>}
      localJointOffset=<3 0 0>
     }]
    simSpaceSettings {simSpaceAngularAlpha=0.02}
   }
   n25 = Anim Dynamics - Bone: dyn_dyn_hair_f_m_02!(ComponentPose=@n24, Alpha=1, GravityOverride:Vector="0,0,-980", ExternalForce="0,0,0") -> (Pose) {
    angularDampingOverride=1
    angularSpringConstant=12000
    bAngularSpring=true
    bOverrideAngularDamping=true
    bUseGravityOverride=true
    gravityScale=0.35
    simulationSpace=RootRelative
    boundBone {boneName=dyn_dyn_hair_f_m_02}
    componentAppliedLinearAccClamp=<5000 5000 5000>
    componentLinearAccScale=<1 1 1>
    componentLinearVelScale=<1 1 1>
    gravityOverride {z=-980}
    physicsBodyDefinitions [{
      boundBone {boneName=dyn_dyn_hair_f_m_02}
      boxExtents=<20 30 30>
      constraintSetup {angularLimitsMax=<0 0 10>, angularLimitsMin=<0 -20 -10>, angularTarget=<1 0 0>}
      localJointOffset=<3 0 0>
     }]
    simSpaceSettings {simSpaceAngularAlpha=0.02}
   }
   n26 = Anim Dynamics - Bone: dyn_hair_f_l_01!(ComponentPose=@n25, Alpha=1, GravityOverride:Vector="0,0,-980", ExternalForce="0,0,0") -> (Pose) {
    angularDampingOverride=1
    angularSpringConstant=12000
    bAngularSpring=true
    bOverrideAngularDamping=true
    bUseGravityOverride=true
    gravityScale=0.35
    simulationSpace=RootRelative
    boundBone {boneName=dyn_hair_f_l_01}
    componentAppliedLinearAccClamp=<5000 5000 5000>
    componentLinearAccScale=<1 1 1>
    componentLinearVelScale=<1 1 1>
    gravityOverride {z=-980}
    physicsBodyDefinitions [{
      boundBone {boneName=dyn_hair_f_l_01}
      boxExtents=<20 30 30>
      constraintSetup {angularLimitsMax=<0 0 10>, angularLimitsMin=<0 -20 -10>, angularTarget=<1 0 0>}
      localJointOffset=<4 0 0>
     }]
    simSpaceSettings {simSpaceAngularAlpha=0.02}
   }
   n27 = Anim Dynamics - Bone: dyn_hair_f_l_02!(ComponentPose=@n26, Alpha=1, GravityOverride:Vector="0,0,-980", ExternalForce="0,0,0") -> (Pose) {
    angularDampingOverride=1
    angularSpringConstant=12000
    bAngularSpring=true
    bOverrideAngularDamping=true
    bUseGravityOverride=true
    gravityScale=0.35
    simulationSpace=RootRelative
    boundBone {boneName=dyn_hair_f_l_02}
    componentAppliedLinearAccClamp=<5000 5000 5000>
    componentLinearAccScale=<1 1 1>
    componentLinearVelScale=<1 1 1>
    gravityOverride {z=-980}
    physicsBodyDefinitions [{
      boundBone {boneName=dyn_hair_f_l_02}
      boxExtents=<20 30 30>
      constraintSetup {angularLimitsMax=<0 0 20>, angularLimitsMin=<0 -15 -20>, angularTarget=<1 0 0>}
      localJointOffset=<4 0 0>
     }]
    simSpaceSettings {simSpaceAngularAlpha=0.02}
   }
   n28 = Anim Dynamics - Bone: dyn_dyn_hair_f_r_01!(ComponentPose=@n27, Alpha=1, GravityOverride:Vector="0,0,-980", ExternalForce="0,0,0") -> (Pose) {
    angularDampingOverride=1
    angularSpringConstant=12000
    bAngularSpring=true
    bOverrideAngularDamping=true
    bUseGravityOverride=true
    gravityScale=0.35
    simulationSpace=RootRelative
    boundBone {boneName=dyn_dyn_hair_f_r_01}
    componentAppliedLinearAccClamp=<5000 5000 5000>
    componentLinearAccScale=<1 1 1>
    componentLinearVelScale=<1 1 1>
    gravityOverride {z=-980}
    physicsBodyDefinitions [{
      boundBone {boneName=dyn_dyn_hair_f_r_01}
      boxExtents=<20 30 30>
      constraintSetup {angularLimitsMax=<0 0 10>, angularLimitsMin=<0 -20 -10>, angularTarget=<1 0 0>}
      localJointOffset=<4 0 0>
     }]
    simSpaceSettings {simSpaceAngularAlpha=0.02}
   }
   n29 = Anim Dynamics - Bone: dyn_dyn_hair_f_r_02!(ComponentPose=@n28, Alpha=1, GravityOverride:Vector="0,0,-980", ExternalForce="0,0,0") -> (Pose) {
    angularDampingOverride=1
    angularSpringConstant=12000
    bAngularSpring=true
    bOverrideAngularDamping=true
    bUseGravityOverride=true
    gravityScale=0.35
    simulationSpace=RootRelative
    boundBone {boneName=dyn_dyn_hair_f_r_02}
    componentAppliedLinearAccClamp=<5000 5000 5000>
    componentLinearAccScale=<1 1 1>
    componentLinearVelScale=<1 1 1>
    gravityOverride {z=-980}
    physicsBodyDefinitions [{
      boundBone {boneName=dyn_dyn_hair_f_r_02}
      boxExtents=<20 30 30>
      constraintSetup {angularLimitsMax=<0 0 10>, angularLimitsMin=<0 -20 -10>, angularTarget=<1 0 0>}
      localJointOffset=<4 0 0>
     }]
    simSpaceSettings {simSpaceAngularAlpha=0.02}
   }
   n30 = Anim Dynamics - Bone: base_hair_r_01(ComponentPose=@n29, Alpha=1, GravityOverride="0,0,-980", ExternalForce="0,0,0") {
    angularDampingOverride=1
    angularSpringConstant=12000
    bAngularSpring=true
    bOverrideAngularDamping=true
    bUseGravityOverride=true
    gravityScale=0.35
    simulationSpace=RootRelative
    boundBone {boneName=base_hair_r_01}
    componentAppliedLinearAccClamp=<5000 5000 5000>
    componentLinearAccScale=<1 1 1>
    componentLinearVelScale=<1 1 1>
    gravityOverride {z=-980}
    physicsBodyDefinitions [{
      boundBone {boneName=base_hair_r_01}
      boxExtents=<20 30 30>
      constraintSetup {angularLimitsMax=<0 20 5>, angularLimitsMin=<0 0 -5>, angularTarget=<1 0 0>}
      localJointOffset=<4 0 0>
     }]
    simSpaceSettings {simSpaceAngularAlpha=0.02}
   }
   n31 = Anim Dynamics - Bone: base_hair_r_02(ComponentPose=@n30, Alpha=1, GravityOverride="0,0,-980", ExternalForce="0,0,0") {
    angularDampingOverride=1
    angularSpringConstant=12000
    bAngularSpring=true
    bOverrideAngularDamping=true
    bUseGravityOverride=true
    gravityScale=0.35
    simulationSpace=RootRelative
    boundBone {boneName=base_hair_r_02}
    componentAppliedLinearAccClamp=<5000 5000 5000>
    componentLinearAccScale=<1 1 1>
    componentLinearVelScale=<1 1 1>
    gravityOverride {z=-980}
    physicsBodyDefinitions [{
      boundBone {boneName=base_hair_r_02}
      boxExtents=<20 30 30>
      constraintSetup {angularLimitsMax=<0 20 5>, angularLimitsMin=<0 0 -5>, angularTarget=<1 0 0>}
      localJointOffset=<4 0 0>
     }]
    simSpaceSettings {simSpaceAngularAlpha=0.02}
   }
   n32 = Anim Dynamics - Bone: dyn_hair_a(ComponentPose=@n31, Alpha=1, ExternalForce="0,0,0") {
    angularDampingOverride=1
    angularSpringConstant=12000
    bAngularSpring=true
    bOverrideAngularDamping=true
    bUseGravityOverride=true
    gravityScale=0.35
    simulationSpace=RootRelative
    boundBone {boneName=dyn_hair_a}
    componentAppliedLinearAccClamp=<5000 5000 5000>
    componentLinearAccScale=<1 1 1>
    componentLinearVelScale=<1 1 1>
    gravityOverride {z=-980}
    physicsBodyDefinitions [{
      boundBone {boneName=dyn_hair_a}
      boxExtents=<20 30 30>
      constraintSetup {angularLimitsMax=<0 0 10>, angularLimitsMin=<0 -10 -10>, angularTarget=<1 0 0>}
      localJointOffset=<2 0 0>
     }]
    simSpaceSettings {simSpaceAngularAlpha=0.2}
   }
   n33 = Anim Dynamics - Bone: dyn_hair_b_01(ComponentPose=@n32, Alpha=1, ExternalForce="0,0,0") {
    angularDampingOverride=1
    angularSpringConstant=12000
    bAngularSpring=true
    bOverrideAngularDamping=true
    bUseGravityOverride=true
    gravityScale=0.35
    simulationSpace=RootRelative
    boundBone {boneName=dyn_hair_b_01}
    componentAppliedLinearAccClamp=<5000 5000 5000>
    componentLinearAccScale=<1 1 1>
    componentLinearVelScale=<1 1 1>
    gravityOverride {z=-980}
    physicsBodyDefinitions [{
      boundBone {boneName=dyn_hair_b_01}
      boxExtents=<20 30 30>
      constraintSetup {angularLimitsMax=<0 0 5>, angularLimitsMin=<0 -10 -5>, angularTarget=<1 0 0>}
      localJointOffset=<2 0 0>
     }]
    simSpaceSettings {simSpaceAngularAlpha=0.2}
   }
   n34 = Anim Dynamics - Bone: dyn_hair_b_02(ComponentPose=@n33, Alpha=1, ExternalForce="0,0,0") {
    angularDampingOverride=1
    angularSpringConstant=12000
    bAngularSpring=true
    bOverrideAngularDamping=true
    bUseGravityOverride=true
    gravityScale=0.35
    simulationSpace=RootRelative
    boundBone {boneName=dyn_hair_b_02}
    componentAppliedLinearAccClamp=<5000 5000 5000>
    componentLinearAccScale=<1 1 1>
    componentLinearVelScale=<1 1 1>
    gravityOverride {z=-980}
    physicsBodyDefinitions [{
      boundBone {boneName=dyn_hair_b_02}
      boxExtents=<20 30 30>
      constraintSetup {angularLimitsMax=<0 0 5>, angularLimitsMin=<0 -10 -5>, angularTarget=<1 0 0>}
      localJointOffset=<2 0 0>
     }]
    simSpaceSettings {simSpaceAngularAlpha=0.2}
   }
   n35 = Anim Dynamics - Bone: dyn_hair_e(ComponentPose=@n34, Alpha=1, ExternalForce="0,0,0") {
    angularDampingOverride=1
    angularSpringConstant=12000
    bAngularSpring=true
    bOverrideAngularDamping=true
    bUseGravityOverride=true
    gravityScale=0.35
    simulationSpace=RootRelative
    boundBone {boneName=dyn_hair_e}
    componentAppliedLinearAccClamp=<5000 5000 5000>
    componentLinearAccScale=<1 1 1>
    componentLinearVelScale=<1 1 1>
    gravityOverride {z=-980}
    physicsBodyDefinitions [{
      boundBone {boneName=dyn_hair_e}
      boxExtents=<20 30 30>
      constraintSetup {angularLimitsMax=<0 0 10>, angularLimitsMin=<0 -10 -10>, angularTarget=<1 0 0>}
      localJointOffset=<4 0 0>
     }]
    simSpaceSettings {simSpaceAngularAlpha=0.2}
   }
   n36 = Anim Dynamics - Bone: dyn_hair_f(ComponentPose=@n35, Alpha=1, ExternalForce="0,0,0") {
    angularDampingOverride=1
    angularSpringConstant=12000
    bAngularSpring=true
    bOverrideAngularDamping=true
    bUseGravityOverride=true
    gravityScale=0.35
    simulationSpace=RootRelative
    boundBone {boneName=dyn_hair_f}
    componentAppliedLinearAccClamp=<5000 5000 5000>
    componentLinearAccScale=<1 1 1>
    componentLinearVelScale=<1 1 1>
    gravityOverride {z=-980}
    physicsBodyDefinitions [{
      boundBone {boneName=dyn_hair_f}
      boxExtents=<20 30 30>
      constraintSetup {angularLimitsMax=<0 0 10>, angularLimitsMin=<0 -10 -10>, angularTarget=<1 0 0>}
      localJointOffset=<4 0 0>
     }]
    simSpaceSettings {simSpaceAngularAlpha=0.2}
   }
   n37 = Anim Dynamics - Bone: dyn_hair_g(ComponentPose=@n36, Alpha=1, ExternalForce="0,0,0") {
    angularDampingOverride=1
    angularSpringConstant=12000
    bAngularSpring=true
    bOverrideAngularDamping=true
    bUseGravityOverride=true
    gravityScale=0.35
    simulationSpace=RootRelative
    boundBone {boneName=dyn_hair_g}
    componentAppliedLinearAccClamp=<5000 5000 5000>
    componentLinearAccScale=<1 1 1>
    componentLinearVelScale=<1 1 1>
    gravityOverride {z=-980}
    physicsBodyDefinitions [{
      boundBone {boneName=dyn_hair_g}
      boxExtents=<20 30 30>
      constraintSetup {angularLimitsMax=<0 0 10>, angularLimitsMin=<0 -10 -10>, angularTarget=<1 0 0>}
      localJointOffset=<4 0 0>
     }]
    simSpaceSettings {simSpaceAngularAlpha=0.2}
   }
   n38 = Anim Dynamics - Bone: dyn_hair_d(ComponentPose=@n37, Alpha=1, ExternalForce="0,0,0") {
    angularDampingOverride=1
    angularSpringConstant=12000
    bAngularSpring=true
    bOverrideAngularDamping=true
    bUseGravityOverride=true
    gravityScale=0.35
    simulationSpace=RootRelative
    boundBone {boneName=dyn_hair_d}
    componentAppliedLinearAccClamp=<5000 5000 5000>
    componentLinearAccScale=<1 1 1>
    componentLinearVelScale=<1 1 1>
    gravityOverride {z=-980}
    physicsBodyDefinitions [{
      boundBone {boneName=dyn_hair_d}
      boxExtents=<20 30 30>
      constraintSetup {angularLimitsMax=<0 0 10>, angularLimitsMin=<0 -10 -10>, angularTarget=<1 0 0>}
      localJointOffset=<4 0 0>
     }]
    simSpaceSettings {simSpaceAngularAlpha=0.2}
   }
   n39 = Anim Dynamics - Bone: dyn_hair_c(ComponentPose=@n38, Alpha=1, ExternalForce="0,0,0") {
    angularDampingOverride=1
    angularSpringConstant=12000
    bAngularSpring=true
    bOverrideAngularDamping=true
    bUseGravityOverride=true
    gravityScale=0.35
    simulationSpace=RootRelative
    boundBone {boneName=dyn_hair_c}
    componentAppliedLinearAccClamp=<5000 5000 5000>
    componentLinearAccScale=<1 1 1>
    componentLinearVelScale=<1 1 1>
    gravityOverride {z=-980}
    physicsBodyDefinitions [{
      boundBone {boneName=dyn_hair_c}
      boxExtents=<20 30 30>
      constraintSetup {angularLimitsMax=<0 0 10>, angularLimitsMin=<0 -10 -10>, angularTarget=<1 0 0>}
      localJointOffset=<4 0 0>
     }]
    simSpaceSettings {simSpaceAngularAlpha=0.2}
   }
   Component To Local(ComponentPose=Local To Component(LocalPose=Component To Local(ComponentPose=@n39)))
   n40 = Get bIsJumping()
   n41 = "Transform (Modify) Bone - Bone: dyn_ponytail_04"!(ComponentPose, bAlphaBoolEnabled:bool=@n40, Translation="0,0,0", Rotation="0,0,0", Scale="1,1,1") -> (Pose) {alphaInputType=Bool, translationMode=BMM_Additive, translationSpace=BCS_BoneSpace, boneToModify {boneName=dyn_ponytail_04}}
   "Transform (Modify) Bone - Bone: dyn_ponytail_05"!(ComponentPose=@n41, bAlphaBoolEnabled:bool=@n40, Translation="0,0,0", Rotation="0,0,0", Scale="1,1,1") -> (Pose) {alphaInputType=Bool, translationMode=BMM_Additive, translationSpace=BCS_BoneSpace, boneToModify {boneName=dyn_ponytail_05}}
   n42 = RigidBody(ComponentPose=Local To Component(LocalPose=Use cached pose 'pre sim_1_1'()), OverrideWorldGravity="0,0,0") {
    bOverrideWorldGravity=true
    overridePhysicsAsset=&F_MED_ChromeSiren_FaceAcc_Physics
    simulationSpace=BaseBoneSpace
    baseBoneRef {boneName=spine_05}
    componentAppliedLinearAccClamp=<1000 1000 1000>
    simSpaceSettings {
     maxAngularAcceleration=50
     maxAngularVelocity=10
     maxLinearAcceleration=1800
     maxLinearVelocity=200
     worldAlpha=1
     externalLinearDragV=<2 2 1>
    }
   }
   n43 = RigidBody(ComponentPose=Local To Component(LocalPose=Use cached pose 'pre sim_1_1'()), OverrideWorldGravity="0,0,0") {
    bOverrideWorldGravity=true
    overridePhysicsAsset=&F_MED_ChromeSiren_FaceAcc_Physics_Cape
    simulationSpace=BaseBoneSpace
    baseBoneRef {boneName=spine_05}
    componentAppliedLinearAccClamp=<1000 1000 1000>
    simSpaceSettings {
     maxAngularAcceleration=50
     maxAngularVelocity=10
     maxLinearAcceleration=1800
     maxLinearVelocity=200
     worldAlpha=1
     externalLinearDragV=<2 2 1>
    }
   }
   n44 = "Transform (Modify) Bone - Bone: dyn_ponytail_02"(ComponentPose=@n43, bAlphaBoolEnabled=Get bIsInFrontEnd(), Translation="0,0,0", Rotation="-3,0,0", Scale="1,1,1") {
    alphaInputType=Bool
    rotationMode=BMM_Additive
    rotationSpace=BCS_BoneSpace
    translationMode=BMM_Additive
    translationSpace=BCS_BoneSpace
    boneToModify {boneName=dyn_ponytail_02}
    rotation {pitch=-3}
   }
   n45 = Blend Poses by bool(
    BlendPose_0=Component To Local(ComponentPose=@n44)
    BlendPose_1=Component To Local(ComponentPose=@n42)
    bActiveValue=Get Is Backbling A Cape()
   ) {blendType=Linear}
   n46 = "Constraint - dyn_ponytail_01 (P)"(ComponentPose=Local To Component(LocalPose=Use cached pose 'pre sim_1_1'()), Alpha=1, ConstraintWeights_0=1, ConstraintWeights_1=0) {boneToModify {boneName=dyn_ponytail_01}, constraintSetup [{targetBone {boneName=head}}], constraintWeights [1,0]}
   n47 = "Constraint - dyn_ponytail_02 (P)"(ComponentPose=@n46, Alpha=1, ConstraintWeights_0=1, ConstraintWeights_1=0) {boneToModify {boneName=dyn_ponytail_02}, constraintSetup [{targetBone {boneName=head}}], constraintWeights [1,0]}
   n48 = "Constraint - dyn_ponytail_03 (P)"(ComponentPose=@n47, Alpha=1, ConstraintWeights_0=1, ConstraintWeights_1=0) {boneToModify {boneName=dyn_ponytail_03}, constraintSetup [{targetBone {boneName=head}}], constraintWeights [1,0]}
   n49 = "Constraint - dyn_ponytail_04 (PPPP)"(ComponentPose=@n48, Alpha=1, ConstraintWeights_0=0.5232, ConstraintWeights_1=0.2939, ConstraintWeights_2=0.1142, ConstraintWeights_3=0.0688) {
    boneToModify {boneName=dyn_ponytail_04}
    constraintSetup [{targetBone {boneName=neck_02}}, {targetBone {boneName=head}}, {targetBone {boneName=spine_05}}, {targetBone {boneName=neck_01}}]
    constraintWeights [1,1,1,1]
   }
   n50 = "Constraint - dyn_ponytail_05 (PPPPP)"(
    ComponentPose=@n49
    Alpha=1
    ConstraintWeights_0=0.4764
    ConstraintWeights_1=0.426
    ConstraintWeights_2=0.0689
    ConstraintWeights_3=0.0186
    ConstraintWeights_4=0.0101
   ) {
    boneToModify {boneName=dyn_ponytail_05}
    constraintSetup [
     {targetBone {boneName=spine_05}}
     {targetBone {boneName=neck_01}}
     {targetBone {boneName=head}}
     {targetBone {boneName=spine_03}}
     {targetBone {boneName=neck_02}}
    ]
    constraintWeights [1,1,1,1,1]
   }
   n51 = "Constraint - dyn_ponytail_06 (PP)"(ComponentPose=@n50, Alpha=1, ConstraintWeights_0=0.9931, ConstraintWeights_1=0.0069) {
    boneToModify {boneName=dyn_ponytail_06}
    constraintSetup [{targetBone {boneName=spine_05}}, {targetBone {boneName=spine_04}}]
    constraintWeights [1,1]
   }
   n52 = "Transform (Modify) Bone - Bone: dyn_ponytail_04"!(ComponentPose=@n51, Alpha:float=1, Translation="0,0,-1", Rotation="0,0,0", Scale="1,1,1") -> (Pose) {translationMode=BMM_Additive, translationSpace=BCS_BoneSpace, boneToModify {boneName=dyn_ponytail_04}, translation {z=-1}}
   n53 = "Transform (Modify) Bone - Bone: dyn_ponytail_05"!(ComponentPose=@n52, Alpha:float=1, Translation="0,0,-2", Rotation="0,0,0", Scale="1,1,1") -> (Pose) {translationMode=BMM_Additive, translationSpace=BCS_BoneSpace, boneToModify {boneName=dyn_ponytail_05}, translation {z=-2}}
   n54 = "Transform (Modify) Bone - Bone: dyn_ponytail_06"(ComponentPose=@n53, Alpha=1, Translation="0,0,-2", Rotation="0,0,0", Scale="1,1,1") {translationMode=BMM_Additive, translationSpace=BCS_BoneSpace, boneToModify {boneName=dyn_ponytail_06}, translation {z=-2}}
   n55 = Component To Local(ComponentPose=@n54)
   n56 = Get is Wear a backpack()
   n57 = Anim Dynamics - Bone: dyn_bangs_l(
    ComponentPose=Local To Component(LocalPose=Blend Poses by bool(BlendPose_0=@n55, BlendPose_1=@n45, bActiveValue=@n56) {blendType=Linear})
    Alpha=1
    ExternalForce="0,0,0"
   ) {
    angularDampingOverride=1
    angularSpringConstant=12000
    bAngularSpring=true
    bOverrideAngularDamping=true
    bUseGravityOverride=true
    gravityScale=0.35
    simulationSpace=RootRelative
    boundBone {boneName=dyn_bangs_l}
    componentLinearAccScale=<0.5 0.5 0.5>
    componentLinearVelScale=<0.5 0.5 0.5>
    gravityOverride {z=-980}
    physicsBodyDefinitions [{
      boundBone {boneName=dyn_bangs_l}
      boxExtents=<20 30 30>
      constraintSetup {angularLimitsMax=<0 8 2>, angularLimitsMin=<0 -8 -2>, angularTarget=<1 0 0>}
      localJointOffset=<1 0 0>
     }]
    simSpaceSettings {simSpaceAngularAlpha=0.5}
   }
   n58 = Anim Dynamics - Bone: dyn_bangs_r(ComponentPose=@n57, Alpha=1, ExternalForce="0,0,0") {
    angularDampingOverride=1
    angularSpringConstant=12000
    bAngularSpring=true
    bOverrideAngularDamping=true
    bUseGravityOverride=true
    gravityScale=0.35
    simulationSpace=RootRelative
    boundBone {boneName=dyn_bangs_r}
    componentLinearAccScale=<0.5 0.5 0.5>
    componentLinearVelScale=<0.5 0.5 0.5>
    gravityOverride {z=-980}
    physicsBodyDefinitions [{
      boundBone {boneName=dyn_bangs_r}
      boxExtents=<20 30 30>
      constraintSetup {angularLimitsMax=<0 8 2>, angularLimitsMin=<0 -8 -2>, angularTarget=<1 0 0>}
      localJointOffset=<1 0 0>
     }]
    simSpaceSettings {simSpaceAngularAlpha=0.5}
   }
   n59 = Anim Dynamics - Bone: dyn_hair_l_01(ComponentPose=@n58, Alpha=1, GravityOverride="0,0,-980", ExternalForce="0,0,0") {
    angularDampingOverride=1
    angularSpringConstant=12000
    bAngularSpring=true
    bOverrideAngularDamping=true
    bUseGravityOverride=true
    gravityScale=0.35
    simulationSpace=RootRelative
    boundBone {boneName=dyn_hair_l_01}
    componentAppliedLinearAccClamp=<5000 5000 5000>
    componentLinearAccScale=<1 1 1>
    componentLinearVelScale=<1 1 1>
    gravityOverride {z=-980}
    physicsBodyDefinitions [{
      boundBone {boneName=dyn_hair_l_01}
      boxExtents=<20 30 30>
      constraintSetup {angularLimitsMax=<0 0 5>, angularLimitsMin=<0 -15 -5>, angularTarget=<1 0 0>}
      localJointOffset=<4 0 0>
     }]
    simSpaceSettings {simSpaceAngularAlpha=0.25}
   }
   n60 = Anim Dynamics - Bone: dyn_hair_l_02(ComponentPose=@n59, Alpha=1, GravityOverride="0,0,-980", ExternalForce="0,0,0") {
    angularDampingOverride=1
    angularSpringConstant=12000
    bAngularSpring=true
    bOverrideAngularDamping=true
    bUseGravityOverride=true
    gravityScale=0.35
    simulationSpace=RootRelative
    boundBone {boneName=dyn_hair_l_02}
    componentAppliedLinearAccClamp=<5000 5000 5000>
    componentLinearAccScale=<1 1 1>
    componentLinearVelScale=<1 1 1>
    gravityOverride {z=-980}
    physicsBodyDefinitions [{
      boundBone {boneName=dyn_hair_l_02}
      boxExtents=<20 30 30>
      constraintSetup {angularLimitsMax=<0 0 5>, angularLimitsMin=<0 -15 -5>, angularTarget=<1 0 0>}
      localJointOffset=<4 0 0>
     }]
    simSpaceSettings {simSpaceAngularAlpha=0.25}
   }
   n61 = Anim Dynamics - Bone: dyn_dyn_hair_f_m_01(ComponentPose=@n60, Alpha=1, GravityOverride="0,0,-980", ExternalForce="0,0,0") {
    angularDampingOverride=1
    angularSpringConstant=12000
    bAngularSpring=true
    bOverrideAngularDamping=true
    bUseGravityOverride=true
    gravityScale=0.35
    simulationSpace=RootRelative
    boundBone {boneName=dyn_dyn_hair_f_m_01}
    componentAppliedLinearAccClamp=<5000 5000 5000>
    componentLinearAccScale=<1 1 1>
    componentLinearVelScale=<1 1 1>
    gravityOverride {z=-980}
    physicsBodyDefinitions [{
      boundBone {boneName=dyn_dyn_hair_f_m_01}
      boxExtents=<20 30 30>
      constraintSetup {angularLimitsMax=<0 0 10>, angularLimitsMin=<0 -20 -10>, angularTarget=<1 0 0>}
      localJointOffset=<3 0 0>
     }]
    simSpaceSettings {simSpaceAngularAlpha=0.25}
   }
   n62 = Anim Dynamics - Bone: dyn_dyn_hair_f_m_02(ComponentPose=@n61, Alpha=1, GravityOverride="0,0,-980", ExternalForce="0,0,0") {
    angularDampingOverride=1
    angularSpringConstant=12000
    bAngularSpring=true
    bOverrideAngularDamping=true
    bUseGravityOverride=true
    gravityScale=0.35
    simulationSpace=RootRelative
    boundBone {boneName=dyn_dyn_hair_f_m_02}
    componentAppliedLinearAccClamp=<5000 5000 5000>
    componentLinearAccScale=<1 1 1>
    componentLinearVelScale=<1 1 1>
    gravityOverride {z=-980}
    physicsBodyDefinitions [{
      boundBone {boneName=dyn_dyn_hair_f_m_02}
      boxExtents=<20 30 30>
      constraintSetup {angularLimitsMax=<0 0 10>, angularLimitsMin=<0 -20 -10>, angularTarget=<1 0 0>}
      localJointOffset=<3 0 0>
     }]
    simSpaceSettings {simSpaceAngularAlpha=0.25}
   }
   n63 = Anim Dynamics - Bone: dyn_hair_f_l_01(ComponentPose=@n62, Alpha=1, GravityOverride="0,0,-980", ExternalForce="0,0,0") {
    angularDampingOverride=1
    angularSpringConstant=12000
    bAngularSpring=true
    bOverrideAngularDamping=true
    bUseGravityOverride=true
    gravityScale=0.35
    simulationSpace=RootRelative
    boundBone {boneName=dyn_hair_f_l_01}
    componentAppliedLinearAccClamp=<5000 5000 5000>
    componentLinearAccScale=<1 1 1>
    componentLinearVelScale=<1 1 1>
    gravityOverride {z=-980}
    physicsBodyDefinitions [{
      boundBone {boneName=dyn_hair_f_l_01}
      boxExtents=<20 30 30>
      constraintSetup {angularLimitsMax=<0 0 10>, angularLimitsMin=<0 -20 -10>, angularTarget=<1 0 0>}
      localJointOffset=<4 0 0>
     }]
    simSpaceSettings {simSpaceAngularAlpha=0.25}
   }
   n64 = Anim Dynamics - Bone: dyn_hair_f_l_02(ComponentPose=@n63, Alpha=1, GravityOverride="0,0,-980", ExternalForce="0,0,0") {
    angularDampingOverride=1
    angularSpringConstant=12000
    bAngularSpring=true
    bOverrideAngularDamping=true
    bUseGravityOverride=true
    gravityScale=0.35
    simulationSpace=RootRelative
    boundBone {boneName=dyn_hair_f_l_02}
    componentAppliedLinearAccClamp=<5000 5000 5000>
    componentLinearAccScale=<1 1 1>
    componentLinearVelScale=<1 1 1>
    gravityOverride {z=-980}
    physicsBodyDefinitions [{
      boundBone {boneName=dyn_hair_f_l_02}
      boxExtents=<20 30 30>
      constraintSetup {angularLimitsMax=<0 0 20>, angularLimitsMin=<0 -15 -20>, angularTarget=<1 0 0>}
      localJointOffset=<4 0 0>
     }]
    simSpaceSettings {simSpaceAngularAlpha=0.25}
   }
   n65 = Anim Dynamics - Bone: dyn_dyn_hair_f_r_01(ComponentPose=@n64, Alpha=1, GravityOverride="0,0,-980", ExternalForce="0,0,0") {
    angularDampingOverride=1
    angularSpringConstant=12000
    bAngularSpring=true
    bOverrideAngularDamping=true
    bUseGravityOverride=true
    gravityScale=0.35
    simulationSpace=RootRelative
    boundBone {boneName=dyn_dyn_hair_f_r_01}
    componentAppliedLinearAccClamp=<5000 5000 5000>
    componentLinearAccScale=<1 1 1>
    componentLinearVelScale=<1 1 1>
    gravityOverride {z=-980}
    physicsBodyDefinitions [{
      boundBone {boneName=dyn_dyn_hair_f_r_01}
      boxExtents=<20 30 30>
      constraintSetup {angularLimitsMax=<0 0 10>, angularLimitsMin=<0 -20 -10>, angularTarget=<1 0 0>}
      localJointOffset=<4 0 0>
     }]
    simSpaceSettings {simSpaceAngularAlpha=0.25}
   }
   n66 = Anim Dynamics - Bone: dyn_dyn_hair_f_r_02(ComponentPose=@n65, Alpha=1, GravityOverride="0,0,-980", ExternalForce="0,0,0") {
    angularDampingOverride=1
    angularSpringConstant=12000
    bAngularSpring=true
    bOverrideAngularDamping=true
    bUseGravityOverride=true
    gravityScale=0.35
    simulationSpace=RootRelative
    boundBone {boneName=dyn_dyn_hair_f_r_02}
    componentAppliedLinearAccClamp=<5000 5000 5000>
    componentLinearAccScale=<1 1 1>
    componentLinearVelScale=<1 1 1>
    gravityOverride {z=-980}
    physicsBodyDefinitions [{
      boundBone {boneName=dyn_dyn_hair_f_r_02}
      boxExtents=<20 30 30>
      constraintSetup {angularLimitsMax=<0 0 10>, angularLimitsMin=<0 -20 -10>, angularTarget=<1 0 0>}
      localJointOffset=<4 0 0>
     }]
    simSpaceSettings {simSpaceAngularAlpha=0.25}
   }
   n67 = Anim Dynamics - Bone: base_hair_r_01(ComponentPose=@n66, Alpha=1, GravityOverride="0,0,-980", ExternalForce="0,0,0") {
    angularDampingOverride=1
    angularSpringConstant=12000
    bAngularSpring=true
    bOverrideAngularDamping=true
    bUseGravityOverride=true
    gravityScale=0.35
    simulationSpace=RootRelative
    boundBone {boneName=base_hair_r_01}
    componentAppliedLinearAccClamp=<5000 5000 5000>
    componentLinearAccScale=<1 1 1>
    componentLinearVelScale=<1 1 1>
    gravityOverride {z=-980}
    physicsBodyDefinitions [{
      boundBone {boneName=base_hair_r_01}
      boxExtents=<20 30 30>
      constraintSetup {angularLimitsMax=<0 20 5>, angularLimitsMin=<0 0 -5>, angularTarget=<1 0 0>}
      localJointOffset=<4 0 0>
     }]
    simSpaceSettings {simSpaceAngularAlpha=0.25}
   }
   n68 = Anim Dynamics - Bone: base_hair_r_02(ComponentPose=@n67, Alpha=1, GravityOverride="0,0,-980", ExternalForce="0,0,0") {
    angularDampingOverride=1
    angularSpringConstant=12000
    bAngularSpring=true
    bOverrideAngularDamping=true
    bUseGravityOverride=true
    gravityScale=0.35
    simulationSpace=RootRelative
    boundBone {boneName=base_hair_r_02}
    componentAppliedLinearAccClamp=<5000 5000 5000>
    componentLinearAccScale=<1 1 1>
    componentLinearVelScale=<1 1 1>
    gravityOverride {z=-980}
    physicsBodyDefinitions [{
      boundBone {boneName=base_hair_r_02}
      boxExtents=<20 30 30>
      constraintSetup {angularLimitsMax=<0 20 5>, angularLimitsMin=<0 0 -5>, angularTarget=<1 0 0>}
      localJointOffset=<4 0 0>
     }]
    simSpaceSettings {simSpaceAngularAlpha=0.25}
   }
   n69 = Anim Dynamics - Bone: dyn_hair_a(ComponentPose=@n68, Alpha=1, ExternalForce="0,0,0") {
    angularDampingOverride=1
    angularSpringConstant=12000
    bAngularSpring=true
    bOverrideAngularDamping=true
    bUseGravityOverride=true
    gravityScale=0.35
    simulationSpace=RootRelative
    boundBone {boneName=dyn_hair_a}
    componentAppliedLinearAccClamp=<5000 5000 5000>
    componentLinearAccScale=<1 1 1>
    componentLinearVelScale=<1 1 1>
    gravityOverride {z=-980}
    physicsBodyDefinitions [{
      boundBone {boneName=dyn_hair_a}
      boxExtents=<20 30 30>
      constraintSetup {angularLimitsMax=<0 0 10>, angularLimitsMin=<0 -10 -10>, angularTarget=<1 0 0>}
      localJointOffset=<2 0 0>
     }]
    simSpaceSettings {simSpaceAngularAlpha=0.25}
   }
   n70 = Anim Dynamics - Bone: dyn_hair_b_01(ComponentPose=@n69, Alpha=1, ExternalForce="0,0,0") {
    angularDampingOverride=1
    angularSpringConstant=12000
    bAngularSpring=true
    bOverrideAngularDamping=true
    bUseGravityOverride=true
    gravityScale=0.35
    simulationSpace=RootRelative
    boundBone {boneName=dyn_hair_b_01}
    componentAppliedLinearAccClamp=<5000 5000 5000>
    componentLinearAccScale=<1 1 1>
    componentLinearVelScale=<1 1 1>
    gravityOverride {z=-980}
    physicsBodyDefinitions [{
      boundBone {boneName=dyn_hair_b_01}
      boxExtents=<20 30 30>
      constraintSetup {angularLimitsMax=<0 0 5>, angularLimitsMin=<0 -10 -5>, angularTarget=<1 0 0>}
      localJointOffset=<2 0 0>
     }]
    simSpaceSettings {simSpaceAngularAlpha=0.25}
   }
   n71 = Anim Dynamics - Bone: dyn_hair_b_02(ComponentPose=@n70, Alpha=1, ExternalForce="0,0,0") {
    angularDampingOverride=1
    angularSpringConstant=12000
    bAngularSpring=true
    bOverrideAngularDamping=true
    bUseGravityOverride=true
    gravityScale=0.35
    simulationSpace=RootRelative
    boundBone {boneName=dyn_hair_b_02}
    componentAppliedLinearAccClamp=<5000 5000 5000>
    componentLinearAccScale=<1 1 1>
    componentLinearVelScale=<1 1 1>
    gravityOverride {z=-980}
    physicsBodyDefinitions [{
      boundBone {boneName=dyn_hair_b_02}
      boxExtents=<20 30 30>
      constraintSetup {angularLimitsMax=<0 0 5>, angularLimitsMin=<0 -10 -5>, angularTarget=<1 0 0>}
      localJointOffset=<2 0 0>
     }]
    simSpaceSettings {simSpaceAngularAlpha=0.25}
   }
   n72 = Anim Dynamics - Bone: dyn_hair_e(ComponentPose=@n71, Alpha=1, ExternalForce="0,0,0") {
    angularDampingOverride=1
    angularSpringConstant=12000
    bAngularSpring=true
    bOverrideAngularDamping=true
    bUseGravityOverride=true
    gravityScale=0.35
    simulationSpace=RootRelative
    boundBone {boneName=dyn_hair_e}
    componentAppliedLinearAccClamp=<5000 5000 5000>
    componentLinearAccScale=<1 1 1>
    componentLinearVelScale=<1 1 1>
    gravityOverride {z=-980}
    physicsBodyDefinitions [{
      boundBone {boneName=dyn_hair_e}
      boxExtents=<20 30 30>
      constraintSetup {angularLimitsMax=<0 0 10>, angularLimitsMin=<0 -10 -10>, angularTarget=<1 0 0>}
      localJointOffset=<4 0 0>
     }]
    simSpaceSettings {simSpaceAngularAlpha=0.25}
   }
   n73 = Anim Dynamics - Bone: dyn_hair_f(ComponentPose=@n72, Alpha=1, ExternalForce="0,0,0") {
    angularDampingOverride=1
    angularSpringConstant=12000
    bAngularSpring=true
    bOverrideAngularDamping=true
    bUseGravityOverride=true
    gravityScale=0.35
    simulationSpace=RootRelative
    boundBone {boneName=dyn_hair_f}
    componentAppliedLinearAccClamp=<5000 5000 5000>
    componentLinearAccScale=<1 1 1>
    componentLinearVelScale=<1 1 1>
    gravityOverride {z=-980}
    physicsBodyDefinitions [{
      boundBone {boneName=dyn_hair_f}
      boxExtents=<20 30 30>
      constraintSetup {angularLimitsMax=<0 0 10>, angularLimitsMin=<0 -10 -10>, angularTarget=<1 0 0>}
      localJointOffset=<4 0 0>
     }]
    simSpaceSettings {simSpaceAngularAlpha=0.25}
   }
   n74 = Anim Dynamics - Bone: dyn_hair_g(ComponentPose=@n73, Alpha=1, ExternalForce="0,0,0") {
    angularDampingOverride=1
    angularSpringConstant=12000
    bAngularSpring=true
    bOverrideAngularDamping=true
    bUseGravityOverride=true
    gravityScale=0.35
    simulationSpace=RootRelative
    boundBone {boneName=dyn_hair_g}
    componentAppliedLinearAccClamp=<5000 5000 5000>
    componentLinearAccScale=<1 1 1>
    componentLinearVelScale=<1 1 1>
    gravityOverride {z=-980}
    physicsBodyDefinitions [{
      boundBone {boneName=dyn_hair_g}
      boxExtents=<20 30 30>
      constraintSetup {angularLimitsMax=<0 0 10>, angularLimitsMin=<0 -10 -10>, angularTarget=<1 0 0>}
      localJointOffset=<4 0 0>
     }]
    simSpaceSettings {simSpaceAngularAlpha=0.25}
   }
   n75 = Anim Dynamics - Bone: dyn_hair_d(ComponentPose=@n74, Alpha=1, ExternalForce="0,0,0") {
    angularDampingOverride=1
    angularSpringConstant=12000
    bAngularSpring=true
    bOverrideAngularDamping=true
    bUseGravityOverride=true
    gravityScale=0.35
    simulationSpace=RootRelative
    boundBone {boneName=dyn_hair_d}
    componentAppliedLinearAccClamp=<5000 5000 5000>
    componentLinearAccScale=<1 1 1>
    componentLinearVelScale=<1 1 1>
    gravityOverride {z=-980}
    physicsBodyDefinitions [{
      boundBone {boneName=dyn_hair_d}
      boxExtents=<20 30 30>
      constraintSetup {angularLimitsMax=<0 0 10>, angularLimitsMin=<0 -10 -10>, angularTarget=<1 0 0>}
      localJointOffset=<4 0 0>
     }]
    simSpaceSettings {simSpaceAngularAlpha=0.25}
   }
   n76 = Anim Dynamics - Bone: dyn_hair_c(ComponentPose=@n75, Alpha=1, ExternalForce="0,0,0") {
    angularDampingOverride=1
    angularSpringConstant=12000
    bAngularSpring=true
    bOverrideAngularDamping=true
    bUseGravityOverride=true
    gravityScale=0.35
    simulationSpace=RootRelative
    boundBone {boneName=dyn_hair_c}
    componentAppliedLinearAccClamp=<5000 5000 5000>
    componentLinearAccScale=<1 1 1>
    componentLinearVelScale=<1 1 1>
    gravityOverride {z=-980}
    physicsBodyDefinitions [{
      boundBone {boneName=dyn_hair_c}
      boxExtents=<20 30 30>
      constraintSetup {angularLimitsMax=<0 0 10>, angularLimitsMin=<0 -10 -10>, angularTarget=<1 0 0>}
      localJointOffset=<4 0 0>
     }]
    simSpaceSettings {simSpaceAngularAlpha=0.25}
   }
   n77 = "Transform (Modify) Bone - Bone: dyn_hair_l_01"!(ComponentPose=@n76, bAlphaBoolEnabled:bool=True, Translation="0,0,0", Rotation="18.973,0,-10", Scale="1,1,1") -> (Pose) {
    alphaInputType=Bool
    rotationMode=BMM_Additive
    translationMode=BMM_Additive
    boneToModify {boneName=dyn_hair_l_01}
    rotation {pitch=18.973, roll=-10}
   }
   Output Pose(Result=Component To Local(ComponentPose=@n77)) {layerGroup=None}
   n78 = Layered blend per bone(BasePose=Copy Pose From Mesh() {bUseAttachedParent=true}, BlendPoses_0=Copy Pose From Mesh(SourceMeshComponent=Get Body_Mesh())) {layerSetup [{branchFilters [
       {boneName=upperarm_l}
       {boneName=upperarm_r}
       {boneName=lowerarm_l}
       {boneName=lowerarm_r}
       {boneName=thigh_l}
       {boneName=thigh_r}
       {boneName=calf_l}
       {boneName=calf_r}
      ]}]}
   Save cached pose 'pre sim_1_1'(Pose=@n78) {cacheName=pre sim_1_1}
   n79 = Layered blend per bone(BasePose=Copy Pose From Mesh() {bUseAttachedParent=true}, BlendPoses_0=Copy Pose From Mesh(SourceMeshComponent=Get Body_Mesh())) {layerSetup [{branchFilters [
       {boneName=upperarm_l}
       {boneName=upperarm_r}
       {boneName=lowerarm_l}
       {boneName=lowerarm_r}
       {boneName=thigh_l}
       {boneName=thigh_r}
       {boneName=calf_l}
       {boneName=calf_r}
      ]}]}
   Save cached pose 'pre sim_1'(Pose=@n79) {cacheName=pre sim_1}
  ]}
 EventGraph {nodes [
   Event BlueprintUpdateAnimation()
   > n1 = Sequence!(execute) -> (then_0, then_1, then_2:Exec, then_3:Exec, then_4:Exec, then_5:Exec)
   > Set Body_Mesh(Body_Mesh=GetOwnerSkeletalMeshComponent())
   n2 = TryGetPawnOwner()
   > Is Valid(InputObject=@n2)
   > n3 = Cast To FortPlayerPawnAthena(Object=@n2)
   > Set As Fort Player Pawn Athena(As Fort Player Pawn Athena:Fort Player Pawn Athena Object Reference=@n3) -> (Output_Get:Fort Player Pawn Athena Object Reference)
   > n4 = Cast To PlayerPawn_Athena(Object=@n2)
   > Set Player Pawn(Player Pawn:Player Pawn Athena Object Reference=@n4) -> (Output_Get:Player Pawn Athena Object Reference)
   n5 = CheckForBackbling()
   n6 = Set IsWearingBackbling(execute=@n1.then_1, IsWearingBackbling:bool=@n5) -> (Output_Get:bool)
   > Set Is Backbling A Cape(Is Backbling A Cape:bool=@n5.bIsBackblingACape) -> (Output_Get:bool)
   > Set is Wear a backpack(is Wear a backpack:bool=AND Boolean(A=@n5, B=NOT Boolean(A=@n5.bIsBackblingACape))) -> (Output_Get:bool)
   > n7 = Branch(Condition=@n6)
   > n8 = Branch(Condition=@n5.bIsBackblingLong)
   Set IsShortBackblingOnAlpha(execute=[@n7.else, @n8.then], IsShortBackblingOnAlpha:double=0) -> (Output_Get:double)
   Set IsShortBackblingOnAlpha(execute=@n8.else, IsShortBackblingOnAlpha=1)
   n9 = "Gravity Override State Function (Parameter Asset) (Nativized)"(
    execute=@n1.then_2
    CharacterPartAnimInstance=Self-Reference()
    GameTimeInSeconds=GetGameTimeInSeconds()
    Parameters=&/F_MED_ChromeSiren_FaceAccParameters
   )
   > Set Gravity Override(Gravity Override=@n9)
   > Set RBAN Settings Out(RBAN Settings Out=@n9.OutRigidBodyAnimNodeParameters)
  ]}
]
variables [
 As Fort Player Pawn Athena {type=FortPlayerPawnAthena}
 Body_Mesh {type=SkeletalMeshComponent}
 Gravity Override {type=Vector}
 Is Backbling A Cape {type=boolean}
 is Wear a backpack {type=boolean}
 IsShortBackblingOnAlpha {type=number}
 IsWearingBackbling {type=boolean}
 Player Pawn {type=PlayerPawn_Athena_C}
 RBAN Settings Out {type=FortRigidBodyAnimNodeParameters}
]
